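; Weapon Blueprint for the Jethro CalmRock assault weapon, handling overheat feedback and a scope rangefinder.
; A heating timer feeds normalized heat to an attached, medallion-aware Niagara effect; overheating switches to smoke emitters with force feedback, while cooling and equip changes reset it.
; Overheat changes drive the scope post-process material's progress.
; While targeting, a timer line-traces from the camera, writing tick-rounded, interpolated distance and out-of-range state to that material.
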(event Custom|SpawnFX
  ;[1 0 0.7965 0.5] Spawn Overheat FX
  (bind _returnvalue (Niagara|SpawnSystemAttached "/JethroWeaponGameplay/Assets/FX/NS_Jethro_OverHeat_Base.NS_Jethro_OverHeat_Base" (Variables|Visual|GetWeaponMesh) "None" "0, 0, 0" "0, 0, 0" "KeepRelativeOffset" false false))
  (bind _output_get (Variables|Default|SetSpawned_Overheat_FX _returnvalue))
  (Effects|Components|ParticleSystem|SetEmitterEnable _output_get "CooldownSmokeCones" false)
  (Effects|Components|ParticleSystem|SetEmitterEnable _output_get "CooldownSmoke" false)
  (Niagara|SetNiagaraVariable(Int32) _output_get "WeaponInt" (Variables|Default|GetWeapon%20Int))
  (Niagara|SetNiagaraVariable(Bool) _output_get "HasMedallion" (GameplayTags|HasTag (GameplayTags|GetOwnedGameplayTags (Variables|Default|GetAs%20Fort%20Player%20Pawn) :!class /Script/GameplayTags.BlueprintGameplayTagLibrary) (Variables|Default|GetMedallionTag))))

(event FortWeapon|EventOnEquippedAndAttached
  (bind _spawned_overheat_fx (Variables|Default|GetSpawned_Overheat_FX))
  ;[0 1 0.4912] On Equip/Unequip -> Reset or else no Overheat FX spawn.
  (|CallParentFunction)
  (CallFunction|ResetFX)
  (Utilities|IsValid _spawned_overheat_fx
    (:Is%20Valid
      (Components|DestroyComponent _spawned_overheat_fx))))

(event FortWeapon|EventOnUnEquippedAndDetached
  (bind _spawned_overheat_fx (Variables|Default|GetSpawned_Overheat_FX))
  ;[0 1 0.4912] On Equip/Unequip -> Reset or else no Overheat FX spawn.
  (|CallParentFunction)
  (CallFunction|ResetFX)
  (Utilities|IsValid _spawned_overheat_fx
    (:Is%20Valid
      (Components|DestroyComponent _spawned_overheat_fx))))

(event Custom|OverheatFX
  (params (Overheat%20Val :type Byte))
  (bind _returnvalue (Math|Conversions|ToInteger(Byte) Overheat%20Val))
  (bind _returnvalue_2 (Math|Conversions|ToFloat(Integer) _returnvalue))
  (bind _returnvalue_3 (/ _returnvalue_2 255.0))
  (bind _spawned_overheat_fx (Variables|Default|GetSpawned_Overheat_FX))
  ;[1 0 0.1652] Overheat FX
  (if (> _returnvalue_3 0.25)
    (Utilities|DoOnceThenDo false
      (:DoOnce
        (CallFunction|SpawnFX))
      (:ThenDo
        _))
    (Components|Activation|Activate _spawned_overheat_fx false)
    (Niagara|SetNiagaraVariable(Float) _spawned_overheat_fx "OverHeat" (Math|Float|Clamp(Float) (- (* _returnvalue_3 2.0) 1.0)))))

(event Custom|ResetFX
  (bind _spawned_overheat_fx (Variables|Default|GetSpawned_Overheat_FX))
  ;[1 0 0.1652] Overheat FX
  (Utilities|DoOnceThenDo false
    (:DoOnce
      (CallFunction|SpawnFX))
    (:ThenDo
      _))
  (Components|Activation|Activate _spawned_overheat_fx false)
  (Niagara|SetNiagaraVariable(Float) _spawned_overheat_fx "OverHeat" (Math|Float|Clamp(Float) (- (* (/ (Math|Conversions|ToFloat(Integer) (Math|Conversions|ToInteger(Byte) (stale Custom|OverheatFX Overheat%20Val))) 255.0) 2.0) 1.0))))

(event EventEndPlay
  (params (EndPlayReason :type EEndPlayReason%20Enum :default "Destroyed"))
  (|CallParentFunction EndPlayReason)
  (Overheating|UnbindEventfromOnOverheatValueChanged (delegate Custom|HandleOverheatValueChanged))
  (EventDispatchers|UnbindEventfromOnTargetingChanged (delegate Custom|HandleDistance)))

(event EventBeginPlay
  (bind _returnvalue (Game|GetInstigator))
  ;[1 0.6311 0.9916 0.5] Event Begin Play Cast to FortPlayerPawn
  (|CallParentFunction)
  (Utilities|IsValid _returnvalue
    (:Is%20Valid
      (bind _asfort_player_pawn (Utilities|Casting|CastToFortPlayerPawn _returnvalue))
      (Variables|Default|SetAs%20Fort%20Player%20Pawn _asfort_player_pawn)
      (Overheating|BindEventtoOnOverheatValueChanged (delegate Custom|HandleOverheatValueChanged))
      (EventDispatchers|BindEventtoOnTargetingChanged (delegate Custom|HandleDistance)))))

(event Custom|HandleOverheatValueChanged
  (params (CurrentValue :type Float%20%28single-precision%29) (MaxValue :type Float%20%28single-precision%29) (State :type EFortWeaponOverheatState%20Enum) (bIsInCooldown :type Boolean))
  ;[1 0.6311 0.9916 0.5] Event Begin Play Cast to FortPlayerPawn
  (CallFunction|UpdateOverheatUI CurrentValue MaxValue State bIsInCooldown))

(event Custom|HandleDistance
  (params (bIsTargeting :type Boolean))
  (CallFunction|UpdateDistance bIsTargeting))

(event Overheating|EventOnStartOverheated
  (bind _spawned_overheat_fx (Variables|Default|GetSpawned_Overheat_FX))
  ;[1 0.1083 0] Overheated
  (|CallParentFunction)
  (Effects|Components|ParticleSystem|SetEmitterEnable _spawned_overheat_fx "CooldownSmokeCones" true)
  (Effects|Components|ParticleSystem|SetEmitterEnable _spawned_overheat_fx "CooldownSmoke" true)
  (Effects|Components|ParticleSystem|SetEmitterEnable _spawned_overheat_fx "WeaponMeshArray" false)
  (Effects|Components|ParticleSystem|SetEmitterEnable _spawned_overheat_fx "Ambient_Steam" false)
  (Effects|Components|ParticleSystem|SetEmitterEnable _spawned_overheat_fx "MedallionFlare" false)
  (Game|Feedback|LocalPlayForceFeedback (Fortnite|GetFortPlayerControllerfromActor (Game|GetInstigator)) "/Game/Weapons/ForceFeedback/FF_Jethro_CalmRock_Overheated.FF_Jethro_CalmRock_Overheated" false "None"))

(event Custom|UpdateOverheatUI
  (params (CurrentValue :type Float%20%28double-precision%29) (MaxValue :type Float%20%28double-precision%29) (State :type EFortWeaponOverheatState%20Enum) (IsInCooldown :type Boolean))
  ;[1 1 1] Overheat UI Support
  (bind _returnvalue (Scope|GetorCreateScopePostProcessMaterialInstanceDynamics))
  (bind _asmaterial_instance_dynamic (Utilities|Casting|CastToMaterialInstanceDynamic (.Object (Utilities|Struct|BreakWeightedBlendable (Utilities|Array|Get(aref) _returnvalue)))))
  (Rendering|Material|SetScalarParameterValue _asmaterial_instance_dynamic "Progress" (/ CurrentValue MaxValue) :!class /Script/Engine.MaterialInstanceDynamic))

(event Overheating|EventOnStartHeating
  ;[1 0.4469 0] Start Heating
  (|CallParentFunction)
  (bind _returnvalue (Utilities|Time|SetTimerbyEvent (EventDispatchers|CreateEvent :!event Custom|Start%20Overheat) (select (Fortnite|IsMobilePlatform) 0.1 0.03) true))
  (Variables|Default|SetOverheatFXTimer _returnvalue))

(event Custom|Start%20Overheat
  ;[1 0.4469 0] Start Heating
  (CallFunction|OverheatFX (Math|Conversions|ToByte(Integer) (Math|Float|Truncate (* 255 (/ (Overheat|GetCurrentOverheatValue) (Overheat|GetOverheatingMaxValue)))))))

(event Overheating|EventOnFinishedCooling
  ;[0 0.6897 1] Finished Cooling
  (|CallParentFunction)
  (Utilities|FlowControl|Delay 0.2)
  (Components|DestroyComponent (Variables|Default|GetSpawned_Overheat_FX))
  (CallFunction|ResetFX))

(event Overheating|EventOnStartCooling
  ;[0 1 0.8693] Start Cooling
  (|CallParentFunction)
  (Utilities|Time|ClearandInvalidateTimerbyHandle (Variables|Default|GetOverheatFXTimer))
  (Utilities|FlowControl|Delay 0.8)
  (Components|Activation|Deactivate (Variables|Default|GetSpawned_Overheat_FX)))

(event Custom|UpdateDistance
  (params (Targeting :type Boolean))
  ;[0 0.4792 0.4248 5] Ditance Update for scope
  (if Targeting
    (bind _returnvalue (Utilities|Time|SetTimerbyEvent (delegate Custom|Timer%20Test) (select (Fortnite|IsMobilePlatform) 0.1 0.03) true))
    (Variables|Default|SetDistanceTimer _returnvalue)
    (else
      (Utilities|Time|ClearandInvalidateTimerbyHandle (Variables|Default|GetDistanceTimer)))))

(event Custom|Timer%20Test
  (bind _scope_distance_ticks (Variables|Default|GetScope_Distance_Ticks))
  (bind _returnvalue (Math|Conversions|ToFloat(Integer) _scope_distance_ticks))
  (bind _distancetimer (Variables|Default|GetDistanceTimer))
  (bind _returnvalue_2 (Utilities|Time|GetTimerElapsedTimebyHandle _distancetimer))
  (bind _scope_distance (Variables|Default|GetScope%20Distance))
  (bind _spawned_overheat_fx (Variables|Default|GetSpawned_Overheat_FX))
  ;[0 0.4792 0.4248 5] Ditance Update for scope
  (if (Weapon|IsUsingScopeorFirstPersonCamera)
    (Utilities|FlowControl|Sequence
      (:then_0
        (bind (_outhit _returnvalue_3) (Collision|LineTraceByChannel (FortWeapon|GetTargetingSourceLocation) (+ (FortWeapon|GetTargetingSourceLocation) (* (Camera|GetCameraRotation (Variables|/Script/Engine.PlayerController|GetPlayerCameraManager (Fortnite|GetFortPlayerControllerfromActor (Variables|Default|GetAs%20Fort%20Player%20Pawn)))) 1e+06)) "TraceTypeQuery1" false (Utilities|Array|MakeArray (Variables|Default|GetAs%20Fort%20Player%20Pawn))))
        (bind _returnvalue_4 (Scope|GetorCreateScopePostProcessMaterialInstanceDynamics))
        (bind _asmaterial_instance_dynamic (Utilities|Casting|CastToMaterialInstanceDynamic (.Object (Utilities|Struct|BreakWeightedBlendable (Utilities|Array|Get(aref) _returnvalue_4)))))
        (if _returnvalue_3
          (Rendering|Material|SetScalarParameterValue _asmaterial_instance_dynamic "OutofRange" 0.0 :!class /Script/Engine.MaterialInstanceDynamic)
          (if (<= (Math|Float|Absolute(Float) (- (Math|Conversions|ToFloat(Integer) (Math|Float|Round (* (Math|Float|Round (/ (/ (.Distance (Collision|BreakHitResult _outhit)) 1e+02) _scope_distance_ticks)) _returnvalue))) (Variables|Default|GetScope%20Distance))) 1e+01)
            (bind _output_get (Variables|Default|SetScope%20Distance (Math|Interpolation|FInterptoConstant _scope_distance (Math|Conversions|ToFloat(Integer) (Math|Float|Round (* (Math|Float|Round (/ (/ (.Distance (Collision|BreakHitResult _outhit)) 1e+02) _scope_distance_ticks)) _returnvalue))) _returnvalue_2 5e+01)))
            (Rendering|Material|SetScalarParameterValue _asmaterial_instance_dynamic "Distance" _output_get :!class /Script/Engine.MaterialInstanceDynamic)
            (else
              (bind _output_get_2 (Variables|Default|SetScope%20Distance (Math|Interpolation|FInterpTo _scope_distance (Math|Conversions|ToFloat(Integer) (Math|Float|Round (* (Math|Float|Round (/ (/ (.Distance (Collision|BreakHitResult _outhit)) 1e+02) _scope_distance_ticks)) _returnvalue))) _returnvalue_2 1e+01)))
              (Rendering|Material|SetScalarParameterValue _asmaterial_instance_dynamic "Distance" _output_get_2 :!class /Script/Engine.MaterialInstanceDynamic)))
          (else
            (Rendering|Material|SetScalarParameterValue _asmaterial_instance_dynamic "OutofRange" 1.0 :!class /Script/Engine.MaterialInstanceDynamic)
            (bind _output_get_3 (Variables|Default|SetScope%20Distance (Math|Interpolation|FInterpTo (Variables|Default|GetScope%20Distance) 1.1e+03 (Utilities|Time|GetWorldDeltaSeconds) 5e+01)))
            (Rendering|Material|SetScalarParameterValue _asmaterial_instance_dynamic "Distance" _output_get_3 :!class /Script/Engine.MaterialInstanceDynamic))))
      (:then_1
        (Effects|Components|ParticleSystem|SetBoolParameter _spawned_overheat_fx "ADS" true)))
    (else
      (Effects|Components|ParticleSystem|SetBoolParameter _spawned_overheat_fx "ADS" false))))

(fn RoundDistance
  (params (Input :type Float%20%28double-precision%29) (Round%20to%20 :type Integer))
  (return 0))

(fn ConstructionScript
  (|CallParentFunction))
